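; Destructible Pretzel Tree log prop for the DesertPoint set that plays a Niagara destruction effect when it dies.
; The death event spawns the medium DesertPoint tree destruction system at the static mesh's location and rotation, with auto-activation off.
; It passes the prop's own mesh and world scale into the system's crystal parameters before activating it, so the effect matches the log.
; The construction script is empty.
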(event EventOnDeathPlayEffects
  (params (Damage :type Float%20%28single-precision%29) (DamageTags :type Gameplay%20Tag%20Container%20Structure%20%28by%20ref%29) (Momentum :type Vector :default "0, 0, 0") (HitInfo :type Hit%20Result%20Structure%20%28by%20ref%29) (InstigatedBy :type Fort%20Pawn%20Object%20Reference) (DamageCauser :type Actor%20Object%20Reference) (EffectContext :type Gameplay%20Effect%20Context%20Handle%20Structure))
  (bind _staticmeshcomponent (Variables|Debug|GetStaticMeshComponent))
  ;[1 1 1] Play crystal death fx
  (bind _returnvalue (Niagara|SpawnSystematLocation "/Game/Effects/Fort_Effects/Effects/Environment/DesertPoint/NS_DesertPoint_Tree_Destruction_M.NS_DesertPoint_Tree_Destruction_M" (Transformation|GetWorldLocation _staticmeshcomponent) (Transformation|GetWorldRotation _staticmeshcomponent) "1.000000,1.000000,1.000000" true false))
  (Niagara|SetNiagaraStaticMeshDirectly _returnvalue "CrystalStaticMesh" (Variables|/Script/Engine.StaticMeshComponent|GetStaticMesh _staticmeshcomponent))
  (Niagara|SetNiagaraVariable(Vector3) _returnvalue "CrystalWorldScale" (Transformation|GetWorldScale (Variables|Debug|GetStaticMeshComponent)))
  (Components|Activation|Activate _returnvalue false))

(fn ConstructionScript)
